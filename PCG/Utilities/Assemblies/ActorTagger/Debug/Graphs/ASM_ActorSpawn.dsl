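; PCG graph that spawns actors and static meshes from assembly points, using each point's Mesh attribute.
; Points are optionally merged, filtered by ActorIndex, and stripped of empty meshes and CULL tags.
; BPActor points go to Spawn Actor; GPUMeshSpawn routes others between two mesh spawners with material and differing component-property overrides.
; OutputPoints gates returned points, which omit the GPUMeshSpawn-enabled branch; the Actors output is unconnected.
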(graph ASM_ActorSpawn
  (params (OutputPoints :type Bool) (MergeInputPoints :type Bool) (GPUMeshSpawn :type Bool))
  (inputs (Points :type Point))
  (outputs (Actors :type Any) (points :type Point))
  (bind Branch_12 (Branch :bOutputToB-param MergeInputPoints))
  (bind MergePoints_14 (Merge%20Points))
  (bind PointFilter_4 (Filter%20Attribute%20Elements :AttributeTypes (struct :Type Integer64 :Int32Value 1 :BoolValue true) :bUseConstantThreshold true :bUseSpatialQuery true :Operator GreaterOrEqual :TargetAttribute ActorIndex))
  (bind PointFilter_17 (Filter%20Attribute%20Elements :AttributeTypes (struct :Type SoftObjectPath :StringValue "None") :bUseConstantThreshold true :bUseSpatialQuery true :Operator Equal :TargetAttribute Mesh)) ; MESH = None
  ;[0 0 0] [TAG] CULL - remove helper objects for example
  (bind PointFilter_9 (Filter%20Attribute%20Elements :AttributeTypes (struct :Type Boolean :Int32Value 1 :BoolValue true) :bUseConstantThreshold true :bUseSpatialQuery true :bWarnOnDataMissingAttribute false :Operator Equal :TargetAttribute CULL)) ; CULL == 1 Remove any points with "CULL" tag 
  ;[0 0 0] Spawning
  (bind AttributeFilter_0 (Filter%20Attribute%20Elements :AttributeTypes (struct :Type Boolean :Int32Value 1 :StringValue "None" :BoolValue true) :bUseConstantThreshold true :bUseSpatialQuery true :bWarnOnDataMissingAttribute false :Operator Equal :TargetAttribute BPActor)) ; BPactorTag
  (bind SpawnActor_0 (Spawn%20Actor :bSpawnByAttribute true :SpawnAttribute Mesh))
  (bind Branch_1 (Branch :bOutputToB-param GPUMeshSpawn))
  (bind StaticMeshSpawner_2 (Static%20Mesh%20Spawner :bAllowDescriptorChanges false :MeshSelectorParameters (class:/Script/PCG.PCGMeshSelectorByAttribute :AttributeName Mesh :bUseAttributeMaterialOverrides true :MaterialOverrideAttributes [Material]) :MeshSelectorType /Script/PCG.PCGMeshSelectorByAttribute :StaticMeshComponentPropertyOverrides [(struct :InputSource bCastShadow :PropertyTarget "bCastShadow") (struct :InputSource WorldPositionOffsetDisableDistance :PropertyTarget "WorldPositionOffsetDisableDistance") (struct :InputSource DetailMode :PropertyTarget "DetailMode") (struct :InputSource bRenderInMainPass :PropertyTarget "bRenderInMainPass") (struct :InputSource bCastHiddenShadow :PropertyTarget "bCastHiddenShadow") (struct :InputSource InstanceStartCullDistance :PropertyTarget "InstanceStartCullDistance") (struct :InputSource InstanceEndCullDistance :PropertyTarget "InstanceEndCullDistance") (struct :InputSource bIncludeInHLOD :PropertyTarget "bIncludeInHLOD") (struct :InputSource bAffectDistanceFieldLighting :PropertyTarget "bAffectDistanceFieldLighting") (struct :InputSource bUseDefaultCollision :PropertyTarget "bUseDefaultCollision")])) ; Has Staticmesh Property Overrides Set
  (bind Branch_10 (Branch :bOutputToB-param OutputPoints)) ; Output Points Switch
  (bind StaticMeshSpawner_0 (Static%20Mesh%20Spawner :bAllowDescriptorChanges false :MeshSelectorParameters (class:/Script/PCG.PCGMeshSelectorByAttribute :AttributeName Mesh :bUseAttributeMaterialOverrides true :MaterialOverrideAttributes [Material]) :MeshSelectorType /Script/PCG.PCGMeshSelectorByAttribute :StaticMeshComponentPropertyOverrides [(struct :InputSource bCastShadow :PropertyTarget "bCastShadow") (struct :InputSource WorldPositionOffsetDisableDistance :PropertyTarget "WorldPositionOffsetDisableDistance") (struct :InputSource DetailMode :PropertyTarget "DetailMode") (struct :InputSource bRenderInMainPass :PropertyTarget "bRenderInMainPass") (struct :InputSource bCastHiddenShadow :PropertyTarget "bCastHiddenShadow") (struct :InputSource InstanceStartCullDistance :PropertyTarget "InstanceStartCullDistance") (struct :InputSource InstanceEndCullDistance :PropertyTarget "InstanceEndCullDistance") (struct :InputSource bAffectDistanceFieldLighting :PropertyTarget "bAffectDistanceFieldLighting") (struct :InputSource bAffectDynamicIndirectLighting :PropertyTarget "bAffectDynamicIndirectLighting") (struct :InputSource bVisibleInRayTracing :PropertyTarget "bVisibleInRayTracing")])) ; Has Staticmesh Property Overrides Set
  (connect PointFilter_9 OutsideFilter AttributeFilter_0 In)
  (connect AttributeFilter_0 OutsideFilter Branch_1 In)
  (connect SpawnActor_0 Out Branch_10 In)
  (connect StaticMeshSpawner_2 Out Branch_10 In)
  (connect DefaultInputNode Points Branch_12 In)
  (connect Branch_12 Output%20B MergePoints_14 In)
  (connect Branch_12 Output%20A PointFilter_4 In)
  (connect MergePoints_14 Out PointFilter_4 In)
  (connect PointFilter_17 OutsideFilter PointFilter_9 In)
  (connect PointFilter_4 InsideFilter PointFilter_17 In)
  (connect AttributeFilter_0 InsideFilter SpawnActor_0 In)
  (connect Branch_1 Output%20B StaticMeshSpawner_0 In)
  (connect Branch_1 Output%20A StaticMeshSpawner_2 In)
  (output points Branch_10.Output%20B))
