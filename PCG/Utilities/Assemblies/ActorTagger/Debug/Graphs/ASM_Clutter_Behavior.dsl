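; PCG subgraph that randomly thins points flagged as Clutter according to a density setting, passing all other points through untouched.
; Clutter points split by ClutterLow, ClutterMed and ClutterHigh flags, with unflagged ones using the base ClutterDensity.
; Each tier offsets that density by a proportion of itself, using its matching parameter.
; Seeded attribute noise filtered against the resulting threshold decides which points survive, then all branches merge.
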(graph ASM_Clutter_Behavior
  (params (ClutterDensity :type Double :default 0.66) (ClutterSeed :type Int64 :default 101) (ClutterLow :type Double :default -0.25) (ClutterMed :type Double :default -0.1) (ClutterHigh :type Double :default 0.2))
  (inputs (Points%20In :type Point))
  (outputs (Points%20Out :type Point))
  (bind AttributeFilter_0 (Filter%20Attribute%20Elements :AttributeTypes (struct :Type Boolean :Int32Value 1 :BoolValue true) :bUseConstantThreshold true :bUseSpatialQuery true :bWarnOnDataMissingAttribute false :Operator Equal :TargetAttribute Clutter)) ; Clutter == 1
  (bind AttributeFilter_7 (Filter%20Attribute%20Elements :AttributeTypes (struct :Type Boolean :Int32Value 1 :IntValue 1 :BoolValue true) :bUseConstantThreshold true :bUseSpatialQuery true :bWarnOnDataMissingAttribute false :Operator Equal :TargetAttribute ClutterLow))
  (bind AttributeFilter_8 (Filter%20Attribute%20Elements :AttributeTypes (struct :Type Boolean :Int32Value 1 :IntValue 1 :BoolValue true) :bUseConstantThreshold true :bUseSpatialQuery true :bWarnOnDataMissingAttribute false :Operator Equal :TargetAttribute ClutterMed))
  (bind AttributeFilter_9 (Filter%20Attribute%20Elements :AttributeTypes (struct :Type Boolean :Int32Value 1 :IntValue 1 :BoolValue true) :bUseConstantThreshold true :bUseSpatialQuery true :bWarnOnDataMissingAttribute false :Operator Equal :TargetAttribute ClutterHigh))
  (bind AttributeNoise_1 (Attribute%20Noise :Seed-param ClutterSeed))
  (bind AddAttribute_6 (Add%20Attribute :AttributeTypes (struct :Type Float) :InputSource ClutterDensity :OutputTarget ClutterDensity :Attributes ClutterDensity))
  (bind AttributeFilter_5 (Filter%20Attribute%20Elements :bUseSpatialQuery true :bWarnOnDataMissingAttribute false :Operator LesserOrEqual :ThresholdAttribute ClutterDensity))
  (bind AttributeNoise_10 (Attribute%20Noise :Seed-param ClutterSeed))
  (bind AttributeMathsOp_16 (Attribute%20Maths%20Op :InputSource1 ClutterDensity :InputSource2 ClutterLow :Operation Multiply :OutputTarget None :InA ClutterDensity :InB ClutterLow))
  (bind AttributeMathsOp_17 (Attribute%20Maths%20Op :InputSource1 ClutterDensity :InputSource2 None))
  (bind DeleteAttributes_18 (Delete%20Attributes :bTokenizeOnWhiteSpace true :Operation KeepSelectedAttributes :SelectedAttributes "ClutterDensity"))
  (bind AddAttribute_14 (Add%20Attribute :AttributeTypes (struct :Type Float) :InputSource ClutterDensity :OutputTarget ClutterDensity))
  (bind AttributeFilter_13 (Filter%20Attribute%20Elements :bUseSpatialQuery true :bWarnOnDataMissingAttribute false :Operator LesserOrEqual :ThresholdAttribute ClutterDensity))
  (bind AttributeNoise_25 (Attribute%20Noise :Seed-param ClutterSeed))
  (bind AttributeMathsOp_22 (Attribute%20Maths%20Op :InputSource1 ClutterDensity :InputSource2 ClutterMed :Operation Multiply :OutputTarget None :InA ClutterDensity :InB ClutterMed))
  (bind AttributeMathsOp_23 (Attribute%20Maths%20Op :InputSource1 ClutterDensity :InputSource2 None))
  (bind DeleteAttributes_24 (Delete%20Attributes :bTokenizeOnWhiteSpace true :Operation KeepSelectedAttributes :SelectedAttributes "ClutterDensity"))
  (bind AddAttribute_21 (Add%20Attribute :AttributeTypes (struct :Type Float) :InputSource ClutterDensity :OutputTarget ClutterDensity))
  (bind AttributeFilter_20 (Filter%20Attribute%20Elements :bUseSpatialQuery true :bWarnOnDataMissingAttribute false :Operator LesserOrEqual :ThresholdAttribute ClutterDensity))
  (bind AttributeNoise_34 (Attribute%20Noise :Seed-param ClutterSeed))
  (bind AttributeMathsOp_31 (Attribute%20Maths%20Op :InputSource1 ClutterDensity :InputSource2 ClutterHigh :Operation Multiply :OutputTarget None :InA ClutterDensity :InB ClutterHigh))
  (bind AttributeMathsOp_32 (Attribute%20Maths%20Op :InputSource1 ClutterDensity :InputSource2 None))
  (bind DeleteAttributes_33 (Delete%20Attributes :bTokenizeOnWhiteSpace true :Operation KeepSelectedAttributes :SelectedAttributes "ClutterDensity"))
  (bind AddAttribute_30 (Add%20Attribute :AttributeTypes (struct :Type Float) :InputSource ClutterDensity :OutputTarget ClutterDensity))
  (bind AttributeFilter_29 (Filter%20Attribute%20Elements :bUseSpatialQuery true :bWarnOnDataMissingAttribute false :Operator LesserOrEqual :ThresholdAttribute ClutterDensity))
  (bind MergePoints_2 (Merge%20Points))
  (connect AttributeNoise_1 Out AddAttribute_6 In)
  (connect DeleteAttributes_18 Out AddAttribute_14 Attributes)
  (connect AttributeNoise_10 Out AddAttribute_14 In)
  (connect DeleteAttributes_24 Out AddAttribute_21 Attributes)
  (connect AttributeNoise_25 Out AddAttribute_21 In)
  (connect DeleteAttributes_33 Out AddAttribute_30 Attributes)
  (connect AttributeNoise_34 Out AddAttribute_30 In)
  (connect DefaultInputNode Points%20In AttributeFilter_0 In)
  (connect AddAttribute_6 Out AttributeFilter_5 In)
  (connect AttributeFilter_0 InsideFilter AttributeFilter_7 In)
  (connect AttributeFilter_7 OutsideFilter AttributeFilter_8 In)
  (connect AttributeFilter_8 OutsideFilter AttributeFilter_9 In)
  (connect AddAttribute_14 Out AttributeFilter_13 In)
  (connect AddAttribute_21 Out AttributeFilter_20 In)
  (connect AddAttribute_30 Out AttributeFilter_29 In)
  (connect AttributeMathsOp_16 Out AttributeMathsOp_17 InA)
  (connect AttributeMathsOp_16 Out AttributeMathsOp_17 InB)
  (connect AttributeMathsOp_22 Out AttributeMathsOp_23 InA)
  (connect AttributeMathsOp_22 Out AttributeMathsOp_23 InB)
  (connect AttributeMathsOp_31 Out AttributeMathsOp_32 InA)
  (connect AttributeMathsOp_31 Out AttributeMathsOp_32 InB)
  (connect AttributeFilter_9 OutsideFilter AttributeNoise_1 In)
  (connect AttributeFilter_7 InsideFilter AttributeNoise_10 In)
  (connect AttributeFilter_8 InsideFilter AttributeNoise_25 In)
  (connect AttributeFilter_9 InsideFilter AttributeNoise_34 In)
  (connect AttributeMathsOp_17 Out DeleteAttributes_18 In)
  (connect AttributeMathsOp_23 Out DeleteAttributes_24 In)
  (connect AttributeMathsOp_32 Out DeleteAttributes_33 In)
  (connect AttributeFilter_0 OutsideFilter MergePoints_2 In)
  (connect AttributeFilter_5 InsideFilter MergePoints_2 In)
  (connect AttributeFilter_13 InsideFilter MergePoints_2 In)
  (connect AttributeFilter_20 InsideFilter MergePoints_2 In)
  (connect AttributeFilter_29 InsideFilter MergePoints_2 In)
  (output Points%20Out MergePoints_2))
