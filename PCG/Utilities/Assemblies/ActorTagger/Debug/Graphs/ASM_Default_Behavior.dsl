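; PCG subgraph applying default tag-driven behavior to assembly points: clutter, random rotation and scale, landscape snap/align, pick and Z offsets.
; Unless NoLandscapeProjection is set, it resolves hierarchy, projects STL/RTL points, slope-filters non-Cliff ones and stores the position delta.
; A final hierarchy pass lets snapped points opt out via STL, then WZAbove0 and ZAboveLand height checks run.
; The SUTL/RUTL branch is unconnected to input and output.
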(graph ASM_Default_Behavior
  (params (ClutterDensity :type Double :default 0.66) (ClutterSeed :type Int64 :default 101) (ClutterLow :type Double :default -0.25) (ClutterMed :type Double :default -0.1) (ClutterHigh :type Double :default 0.2) (RecomputeSeed :type Bool) (Enable_ASMTextDebug :type Bool) (NoLandscapeProjection :type Bool))
  (inputs (Points :type Point))
  (outputs (Points%20Out :type Point))
  (bind CopyAttribute_0 (Copy%20Attributes :InputSource $Index :OutputTarget PtID))
  (bind CreateAttribute_89 (Create%20Constant :AttributeTypes (struct :Type Vector) :OutputTarget STLDelta))
  (bind CopyAttribute_91 (Copy%20Attributes :InputSource STLDelta :OutputTarget STLDelta))
  (bind CopyAttribute_93 (Copy%20Attributes :InputSource $Position :OutputTarget OldPos))
  (bind MergePoints_44 (Merge%20Points))
  (bind Gather_1 (Gather))
  ;[0 0 0] Clutter [TAG] Clutter [TAG] ClutterLow [TAG] ClutterMed [TAG] ClutterHigh
  (bind Subgraph_80 (Subgraph :Graph /PCG/Utilities/Assemblies/ActorTagger/Debug/Graphs/ASM_Clutter_Behavior.ASM_Clutter_Behavior :ClutterDensity ClutterDensity :ClutterHigh ClutterHigh :ClutterLow ClutterLow :ClutterMed ClutterMed :ClutterSeed ClutterSeed))
  ;[0 0 0] [TAG] RotZ
  (bind PointFilter_11 (Filter%20Attribute%20Elements :AttributeTypes (struct :Type Boolean :Int32Value 1 :BoolValue true) :bUseConstantThreshold true :bUseSpatialQuery true :Operator Equal :TargetAttribute RotZ)) ; RotZ == 1
  (bind TransformPoints_13 (Transform%20Points :AttributeName RelativeTransform :bApplyToAttribute true :RotationMax (struct :Yaw 1.8e+02) :RotationMin (struct :Yaw -1.8e+02)))
  ;[0 0 0] [TAG] RotX
  (bind PointFilter_45 (Filter%20Attribute%20Elements :AttributeTypes (struct :Type Boolean :Int32Value 1 :IntValue 1 :BoolValue true) :bUseConstantThreshold true :bUseSpatialQuery true :Operator Equal :TargetAttribute RotX)) ; RotX == 1
  (bind TransformPoints_16 (Transform%20Points :AttributeName RelativeTransform :bApplyToAttribute true :RotationMax (struct :Roll 1.8e+02) :RotationMin (struct :Roll -1.8e+02)))
  ;[0 0 0] [TAG] RotY
  (bind AttributeFilter_5 (Filter%20Attribute%20Elements :AttributeTypes (struct :Type Boolean :Int32Value 1 :IntValue 1 :BoolValue true) :bUseConstantThreshold true :bUseSpatialQuery true :Operator Equal :TargetAttribute RotY)) ; RotY == 1
  (bind TransformPoints_3 (Transform%20Points :AttributeName RelativeTransform :bApplyToAttribute true :RotationMax (struct :Pitch 1.8e+02) :RotationMin (struct :Pitch -1.8e+02)))
  ;[0 0 0] Tag[FixedScale] and other scale based tags BP Random Scale Ranges
  (bind PointFilter_6 (Filter%20Attribute%20Elements :AttributeTypes (struct :Type Boolean :Int32Value 1 :IntValue 1 :BoolValue true) :bUseConstantThreshold true :bUseSpatialQuery true :Operator Equal :TargetAttribute FixedScale)) ; [TAG]FixedScale
  (bind PointFilter_59 (Filter%20Attribute%20Elements :AttributeTypes (struct :Type SoftObjectPath :StringValue "None") :bUseConstantThreshold true :bUseSpatialQuery true :Operator Equal :TargetAttribute Mesh :ThresholdAttribute None)) ; Mesh == None
  (bind PointFilter_22 (Filter%20Attribute%20Elements :AttributeTypes (struct :Type Boolean :Int32Value 1 :IntValue 1 :BoolValue true) :bUseConstantThreshold true :bUseSpatialQuery true :Operator Equal :TargetAttribute RandScale5)) ; [TAG]RandScale5
  (bind PointFilter_4 (Filter%20Attribute%20Elements :AttributeTypes (struct :Type Boolean :Int32Value 1 :IntValue 1 :BoolValue true) :bUseConstantThreshold true :bUseSpatialQuery true :Operator Equal :TargetAttribute RandScale10)) ; [TAG]RandScale10
  (bind AttributeNoise_6 (Attribute%20Noise :NoiseMax 1.1 :NoiseMin 0.9 :OutputTarget RandScale)) ; Create RandScale 0.9 to 1.1
  (bind PointFilter_10 (Filter%20Attribute%20Elements :AttributeTypes (struct :Type Boolean :Int32Value 1 :IntValue 1 :BoolValue true) :bUseConstantThreshold true :bUseSpatialQuery true :Operator Equal :TargetAttribute RandScale15)) ; [TAG]RandScale15
  (bind AttributeNoise_12 (Attribute%20Noise :NoiseMax 1.15 :NoiseMin 0.85 :OutputTarget RandScale)) ; Create RandScale 0.85 to 1.15
  (bind PointFilter_16 (Filter%20Attribute%20Elements :AttributeTypes (struct :Type Boolean :Int32Value 1 :IntValue 1 :BoolValue true) :bUseConstantThreshold true :bUseSpatialQuery true :Operator Equal :TargetAttribute RandScale20)) ; [TAG]RandScale20
  (bind AttributeNoise_18 (Attribute%20Noise :NoiseMax 1.2 :NoiseMin 0.8 :OutputTarget RandScale)) ; Create RandScale 0.8 to 1.2
  (bind AttributeNoise_24 (Attribute%20Noise :NoiseMax 1.05 :NoiseMin 0.95 :OutputTarget RandScale)) ; Create RandScale 0.95 to 1.05
  (bind Merge_57 (Merge%20Points))
  (bind AttributeMathsOp_26 (Attribute%20Maths%20Op :InputSource1 RelativeTransform.Scale :InputSource2 RandScale :Operation Multiply :OutputTarget RelativeTransform.Scale)) ; Apply RandScale to RelativeTransform.Scale
  (bind Merge_2 (Merge%20Points))
  ;[0 0 0] No Landscape Projection Bool - Branch
  (bind Branch_60 (Branch :bOutputToB true :bOutputToB-param NoLandscapeProjection))
  ;[0 0 0] Apply Hierarchy to apply transforms before projecting on landscape
  (bind ApplyHierarchy_0 (Apply%20Hierarchy :ApplyHierarchyAttribute STL :ApplyParentRotationAttribute IPR :ApplyParentScaleAttribute IPS :bWarnOnPointsWithInvalidParent false))
  ;[0 0 0] [STL] [RTL] Landscape Snapping and Align.
  (bind PointFilter_47 (Filter%20Attribute%20Elements :AttributeTypes (struct :Type Boolean :Int32Value 1 :BoolValue true) :bUseConstantThreshold true :bUseSpatialQuery true :Operator Equal :TargetAttribute RTL)) ; RTL == 1
  (bind PointFilter_49 (Filter%20Attribute%20Elements :AttributeTypes (struct :Type Boolean :Int32Value 1 :BoolValue true) :bUseConstantThreshold true :bUseSpatialQuery true :Operator Equal :TargetAttribute STL)) ; STL == 1
  (bind MergePoints_3 (Merge%20Points))
  (bind CopyAttribute_7 (Copy%20Attributes :InputSource $Transform.Rotation :OutputTarget RotBak))
  (bind AttributeFilter_12 (Filter%20Attribute%20Elements :AttributeTypes (struct :Type Boolean :BoolValue true) :bUseConstantThreshold true :bUseSpatialQuery true :Operator Equal :TargetAttribute RTL)) ; RTL == 1
  (bind CopyAttribute_10 (Copy%20Attributes :InputSource RotBak :OutputTarget $Transform.Rotation)) ; STL points copy back rotations
  (bind AttributeFilter_6 (Filter%20Attribute%20Elements :AttributeTypes (struct :Type Boolean :Int32Value 1 :BoolValue true) :bUseConstantThreshold true :bUseSpatialQuery true :Operator Equal :TargetAttribute YUp)) ; YUP == 1
  (bind MakeRotatorAttribute_7 (Make%20Rotator%20Attribute :InputSource1 $Rotation.Up :Operation MakeRotFromY :OutputTarget $Rotation))
  (bind TransformPoints_1 (Transform%20Points :AttributeName RelativeTransform :RotationMax (struct :Pitch 1.8e+02) :RotationMin (struct :Pitch -1.8e+02)))
  (bind AttributeFilter_9 (Filter%20Attribute%20Elements :AttributeTypes (struct :Type Boolean :Int32Value 1 :BoolValue true) :bUseConstantThreshold true :bUseSpatialQuery true :Operator Equal :TargetAttribute XUp)) ; XUp == 1
  (bind MakeRotatorAttribute_13 (Make%20Rotator%20Attribute :InputSource1 $Rotation.Up :Operation MakeRotFromX :OutputTarget $Rotation))
  (bind TransformPoints_5 (Transform%20Points :AttributeName RelativeTransform :RotationMax (struct :Roll 1.8e+02) :RotationMin (struct :Roll -1.8e+02)))
  (bind AttributeFilter_16 (Filter%20Attribute%20Elements :AttributeTypes (struct :Type Boolean :Int32Value 1 :BoolValue true) :bUseConstantThreshold true :bUseSpatialQuery true :Operator Equal :TargetAttribute ZDown)) ; ZDown == 1
  (bind TransformPoints_23 (Transform%20Points :AttributeName RelativeTransform :RotationMax (struct :Pitch 1.8e+02) :RotationMin (struct :Pitch 1.8e+02)))
  (bind AttributeFilter_30 (Filter%20Attribute%20Elements :AttributeTypes (struct :Type Boolean :Int32Value 1 :BoolValue true) :bUseConstantThreshold true :bUseSpatialQuery true :Operator Equal :TargetAttribute RotZ)) ; RotZ == 1
  (bind TransformPoints_28 (Transform%20Points :AttributeName RelativeTransform :bApplyToAttribute true :RotationMax (struct :Yaw 1.8e+02) :RotationMin (struct :Yaw -1.8e+02)))
  (bind AttributeFilter_17 (Filter%20Attribute%20Elements :AttributeTypes (struct :Type Boolean :Int32Value 1 :BoolValue true) :bUseConstantThreshold true :bUseSpatialQuery true :Operator Equal :TargetAttribute RotY)) ; RotY == 1
  (bind TransformPoints_2 (Transform%20Points :AttributeName RelativeTransform :RotationMax (struct :Pitch 1.8e+02) :RotationMin (struct :Pitch -1.8e+02)))
  (bind AttributeFilter_22 (Filter%20Attribute%20Elements :AttributeTypes (struct :Type Boolean :Int32Value 1 :BoolValue true) :bUseConstantThreshold true :bUseSpatialQuery true :Operator Equal :TargetAttribute RotX)) ; RotX == 1
  (bind CreateAttribute_79 (Create%20Constant :AttributeTypes (struct :Type Boolean :Int32Value 1 :IntValue 1 :BoolValue true) :OutputTarget STL))
  (bind TransformPoints_19 (Transform%20Points :AttributeName RelativeTransform :RotationMax (struct :Roll 1.8e+02) :RotationMin (struct :Roll -1.8e+02)))
  (bind CopyAttribute_78 (Copy%20Attributes :InputSource STL)) ; We set STL to 1 so that the Apply Hierarchy node knows which \nchildren points should not inherit a parents transform
  (bind Merge_3 (Merge%20Points))
  ;;[0 0 0] RTL and STL
  (bind Projection_41 (Projection)) ; RTL
  (bind PointFilter_36 (Filter%20Attribute%20Elements :AttributeTypes (struct :Type Boolean) :bUseConstantThreshold true :bUseSpatialQuery true :Operator Equal :TargetAttribute Cliff)) ; Cliff==0
  (bind NormalToDensity_31 (Normal%20To%20Density :Offset -0.8 :Strength 14.0))
  (bind DensityFilter_29 (Density%20Filter))
  (bind NormalToDensity_29 (Normal%20To%20Density :Offset -0.8 :Strength 14.0))
  (bind Merge_37 (Merge%20Points))
  (bind AttributeMathsOp_81 (Attribute%20Maths%20Op :InputSource1 $Position :InputSource2 OldPos :Operation Subtract :OutputTarget STLDelta))
  (bind AttributeMathsOp_85 (Attribute%20Maths%20Op :InputSource1 STLDelta :InputSource2 RelativeTransform.Position :OutputTarget RelativeTransform.Position))
  ;[0 0 0] [TAG] Pick1 Pick2 Pick3 Pick4 PickLocal
  (bind Subgraph_0 (Subgraph :Graph /PCG/Utilities/Assemblies/ActorTagger/Debug/Graphs/ASM_Pick_Behavior.ASM_Pick_Behavior))
  ;[0 0 0] [TAG] FloatZ [TAG] OffsetZ
  (bind PointFilter_24 (Filter%20Attribute%20Elements :AttributeTypes (struct :Int32Value 1) :bUseConstantThreshold true :bUseSpatialQuery true :Operator NotEqual :TargetAttribute FloatZ)) ; PFilter_FloatZ != 0.0
  (bind AttributeFilter_1 (Filter%20Attribute%20Elements :AttributeTypes (struct :Int32Value 1 :IntValue 1 :BoolValue true) :bUseConstantThreshold true :Operator NotEqual :TargetAttribute OffsetZ)) ; [TAG] OffsetZ !=0.0
  (bind PointFilter_1 (Filter%20Attribute%20Elements :AttributeTypes (struct :Int32Value 1 :IntValue 1 :BoolValue true) :bUseConstantThreshold true :Operator NotEqual :TargetAttribute OffsetZ)) ; [TAG] OffsetZ !=0.0
  (bind AttributeFilter_8 (Filter%20Attribute%20Elements :AttributeTypes (struct :Type Boolean :Int32Value 1 :BoolValue true) :bUseConstantThreshold true :bUseSpatialQuery true :Operator Equal :TargetAttribute YUp))
  (bind AttributeFilter_10 (Filter%20Attribute%20Elements :AttributeTypes (struct :Type Boolean :Int32Value 1 :BoolValue true) :bUseConstantThreshold true :bUseSpatialQuery true :Operator Equal :TargetAttribute XUp))
  (bind AttributeMathsOp_7 (Attribute%20Maths%20Op :InputSource1 $Position.Z :InputSource2 OffsetZ :OutputTarget $Position.Z))
  (bind AttributeNoise_13 (Attribute%20Noise))
  (bind AttributeMathsOp_15 (Attribute%20Maths%20Op :InputSource1 FloatZ :InputSource2 $Density :Operation Multiply))
  (bind AttributeFilter_24 (Filter%20Attribute%20Elements :AttributeTypes (struct :Type Boolean :Int32Value 1 :BoolValue true) :bUseConstantThreshold true :bUseSpatialQuery true :Operator Equal :TargetAttribute YUp))
  (bind AttributeFilter_27 (Filter%20Attribute%20Elements :AttributeTypes (struct :Type Boolean :Int32Value 1 :BoolValue true) :bUseConstantThreshold true :bUseSpatialQuery true :Operator Equal :TargetAttribute XUp))
  (bind AttributeMathsOp_12 (Attribute%20Maths%20Op :InputSource1 $Position.Z :InputSource2 FloatZ :OutputTarget $Position.Z))
  (bind AttributeMathsOp_18 (Attribute%20Maths%20Op :InputSource1 $Position.Y :InputSource2 OffsetZ :OutputTarget $Position.Y))
  (bind AttributeMathsOp_21 (Attribute%20Maths%20Op :InputSource1 $Position.X :InputSource2 OffsetZ :OutputTarget $Position.X))
  (bind AttributeMathsOp_29 (Attribute%20Maths%20Op :InputSource1 $Position.Y :InputSource2 FloatZ :OutputTarget $Position.Y))
  (bind AttributeMathsOp_31 (Attribute%20Maths%20Op :InputSource1 $Position.X :InputSource2 FloatZ :OutputTarget $Position.X))
  (bind Merge_25 (Merge%20Points))
  ;[0 0 0] No Landscape Projection Bool - Branch
  (bind Branch_10 (Branch :bOutputToB true :bOutputToB-param NoLandscapeProjection))
  (bind ApplyHierarchy_1 (Apply%20Hierarchy :ApplyHierarchyAttribute STL :ApplyParentRotationAttribute IPR :ApplyParentScaleAttribute IPS :bWarnOnPointsWithInvalidParent false))
  (bind ApplyHierarchy_2 (Apply%20Hierarchy :ApplyHierarchy OptOutByAttribute :ApplyHierarchyAttribute STL :ApplyParentRotationAttribute IPR :ApplyParentScaleAttribute IPS :bWarnOnPointsWithInvalidParent false)) ; We use the STL attribute we set earlier \nto opt out of applying \nthe hierarchy transforms
  (bind Gather_0 (Gather))
  ;[0.243 1 0.2978 1 36] Landscape Input
  (bind GetActorData_37 (Get%20Actor%20Data :ActorSelector (struct :ActorFilter AllWorldActors :ActorSelection ByClass :ActorSelectionClass /Script/Landscape.Landscape) :bComponentsMustOverlapSelf false :bTrackActorsOnlyWithinBounds false))
  ;[0 0 0] [SUTL] and [RUTL] tags
  (bind AttributeFilter_118 (Filter%20Attribute%20Elements :AttributeTypes (struct :Type Boolean :BoolValue true) :bUseConstantThreshold true :bUseSpatialQuery true :Operator Equal :TargetAttribute SUTL)) ; SUTL > 0
  (bind AttributeFilter_11 (Filter%20Attribute%20Elements :AttributeTypes (struct :Type Boolean :BoolValue true) :bUseConstantThreshold true :bUseSpatialQuery true :Operator Equal :TargetAttribute RUTL)) ; RUTL > 0
  (bind CopyAttribute_9 (Copy%20Attributes :InputSource $Position :OutputTarget OldPos))
  (bind CopyAttribute_14 (Copy%20Attributes :InputSource $Transform.Rotation :OutputTarget RotBak))
  (bind Projection_13 (Projection)) ; RUTL and SUTL
  (bind AttributeFilter_110 (Filter%20Attribute%20Elements :TargetAttribute $Position.Z :ThresholdAttribute OldPos.Z))
  (bind CopyAttribute_112 (Copy%20Attributes :InputSource OldPos :OutputTarget $Position))
  (bind AttributeMathsOp_2 (Attribute%20Maths%20Op :InputSource1 $Position.Z :InputSource2 FloatLocalZ))
  (bind AttributeFilter_0 (Filter%20Attribute%20Elements :AttributeTypes (struct :Type Boolean) :bUseConstantThreshold true :bUseSpatialQuery true :Operator Equal :TargetAttribute Cliff))
  (bind AttributeFilter_20 (Filter%20Attribute%20Elements :AttributeTypes (struct :Type Boolean :BoolValue true) :bUseConstantThreshold true :bUseSpatialQuery true :Operator Equal :TargetAttribute RUTL)) ; RUTL > 0
  (bind CopyAttribute_16 (Copy%20Attributes :InputSource RotBak :OutputTarget $Rotation)) ; SUTL
  ;;[0 0 0 1 25] No Cliff Tag
  (bind NormalToDensity_4 (Normal%20To%20Density :Offset -0.8 :Strength 14.0))
  (bind DensityFilter_2 (Density%20Filter))
  ;[0 0 0] [WZAbove0]   WZ Above Zero tags
  (bind AttributeFilter_2 (Filter%20Attribute%20Elements :AttributeTypes (struct :Type Boolean :BoolValue true) :bUseConstantThreshold true :bUseSpatialQuery true :Operator Equal :TargetAttribute WZAbove0 :ThresholdAttribute @Last))
  (bind AttributeFilter_4 (Filter%20Attribute%20Elements :bUseConstantThreshold true :bUseSpatialQuery true :TargetAttribute $Position.Z :ThresholdAttribute @Last))
  ;[0 0 0] ZAboveLand tags
  (bind AttributeFilter_14 (Filter%20Attribute%20Elements :AttributeTypes (struct :Type Boolean :BoolValue true) :bUseConstantThreshold true :bUseSpatialQuery true :Operator Equal :TargetAttribute ZAboveLand :ThresholdAttribute @Last))
  (bind CopyAttribute_30 (Copy%20Attributes :InputSource CULL :OutputTarget Temp))
  (bind Projection_18 (Projection)) ; ZAboveLand
  (bind AttributeCompareOp_22 (Attribute%20Compare%20Op :InputSource1 $Position.Z :InputSource2 $Position.Z :Operation GreaterOrEqual :OutputTarget Temp))
  (reroute Landscape%20Input GetActorData_37)
  (connect Branch_60 Output%20A ApplyHierarchy_0 In)
  (connect Branch_10 Output%20B ApplyHierarchy_1 In)
  (connect Branch_10 Output%20A ApplyHierarchy_2 In)
  (connect Projection_18 Out AttributeCompareOp_22 InA)
  (connect CopyAttribute_30 Out AttributeCompareOp_22 InB)
  (connect AttributeMathsOp_2 Out AttributeFilter_0 In)
  (connect PointFilter_24 OutsideFilter AttributeFilter_1 In)
  (connect Gather_0 Out AttributeFilter_2 In)
  (connect AttributeFilter_2 InsideFilter AttributeFilter_4 In)
  (connect PointFilter_45 OutsideFilter AttributeFilter_5 In)
  (connect TransformPoints_16 Out AttributeFilter_5 In)
  (connect AttributeFilter_12 InsideFilter AttributeFilter_6 In)
  (connect AttributeFilter_1 InsideFilter AttributeFilter_8 In)
  (connect PointFilter_1 InsideFilter AttributeFilter_8 In)
  (connect AttributeFilter_6 OutsideFilter AttributeFilter_9 In)
  (connect TransformPoints_1 Out AttributeFilter_9 In)
  (connect AttributeFilter_8 OutsideFilter AttributeFilter_10 In)
  (connect AttributeFilter_118 OutsideFilter AttributeFilter_11 In)
  (connect AttributeMathsOp_85 Out AttributeFilter_12 In)
  (connect AttributeFilter_2 OutsideFilter AttributeFilter_14 In)
  (connect AttributeFilter_4 InsideFilter AttributeFilter_14 In)
  (connect AttributeFilter_9 OutsideFilter AttributeFilter_16 In)
  (connect TransformPoints_5 Out AttributeFilter_16 In)
  (connect AttributeFilter_30 OutsideFilter AttributeFilter_17 In)
  (connect TransformPoints_28 Out AttributeFilter_17 In)
  (connect AttributeFilter_0 OutsideFilter AttributeFilter_20 In)
  (connect DensityFilter_2 Out AttributeFilter_20 In)
  (connect AttributeFilter_17 OutsideFilter AttributeFilter_22 In)
  (connect TransformPoints_2 Out AttributeFilter_22 In)
  (connect AttributeMathsOp_15 Out AttributeFilter_24 In)
  (connect AttributeFilter_24 OutsideFilter AttributeFilter_27 In)
  (connect AttributeFilter_16 OutsideFilter AttributeFilter_30 In)
  (connect TransformPoints_23 Out AttributeFilter_30 In)
  (connect Projection_13 Out AttributeFilter_110 In)
  (connect AttributeFilter_110 InsideFilter AttributeMathsOp_2 InA)
  (connect AttributeFilter_110 InsideFilter AttributeMathsOp_2 InB)
  (connect AttributeFilter_10 OutsideFilter AttributeMathsOp_7 InA)
  (connect AttributeFilter_10 OutsideFilter AttributeMathsOp_7 InB)
  (connect AttributeFilter_27 OutsideFilter AttributeMathsOp_12 InA)
  (connect AttributeFilter_27 OutsideFilter AttributeMathsOp_12 InB)
  (connect AttributeNoise_13 Out AttributeMathsOp_15 InA)
  (connect AttributeNoise_13 Out AttributeMathsOp_15 InB)
  (connect AttributeFilter_8 InsideFilter AttributeMathsOp_18 InA)
  (connect AttributeFilter_8 InsideFilter AttributeMathsOp_18 InB)
  (connect AttributeFilter_10 InsideFilter AttributeMathsOp_21 InA)
  (connect AttributeFilter_10 InsideFilter AttributeMathsOp_21 InB)
  (connect Merge_57 Out AttributeMathsOp_26 InA)
  (connect Merge_57 Out AttributeMathsOp_26 InB)
  (connect AttributeFilter_24 InsideFilter AttributeMathsOp_29 InA)
  (connect AttributeFilter_24 InsideFilter AttributeMathsOp_29 InB)
  (connect AttributeFilter_27 InsideFilter AttributeMathsOp_31 InA)
  (connect AttributeFilter_27 InsideFilter AttributeMathsOp_31 InB)
  (connect Merge_37 Out AttributeMathsOp_81 InA)
  (connect Merge_37 Out AttributeMathsOp_81 InB)
  (connect AttributeMathsOp_81 Out AttributeMathsOp_85 InA)
  (connect AttributeMathsOp_81 Out AttributeMathsOp_85 InB)
  (connect PointFilter_4 InsideFilter AttributeNoise_6 In)
  (connect PointFilter_10 InsideFilter AttributeNoise_12 In)
  (connect PointFilter_1 OutsideFilter AttributeNoise_13 In)
  (connect PointFilter_16 InsideFilter AttributeNoise_18 In)
  (connect PointFilter_22 InsideFilter AttributeNoise_24 In)
  (connect Merge_25 Out Branch_10 In)
  (connect Merge_2 Out Branch_60 In)
  (connect DefaultInputNode Points CopyAttribute_0 Source)
  (connect DefaultInputNode Points CopyAttribute_0 Target)
  (connect MergePoints_3 Out CopyAttribute_7 Source)
  (connect MergePoints_3 Out CopyAttribute_7 Target)
  (connect AttributeFilter_11 InsideFilter CopyAttribute_9 Source)
  (connect AttributeFilter_118 InsideFilter CopyAttribute_9 Source)
  (connect AttributeFilter_11 InsideFilter CopyAttribute_9 Target)
  (connect AttributeFilter_118 InsideFilter CopyAttribute_9 Target)
  (connect AttributeFilter_12 OutsideFilter CopyAttribute_10 Source)
  (connect AttributeFilter_12 OutsideFilter CopyAttribute_10 Target)
  (connect CopyAttribute_9 Out CopyAttribute_14 Source)
  (connect CopyAttribute_9 Out CopyAttribute_14 Target)
  (connect AttributeFilter_20 OutsideFilter CopyAttribute_16 Source)
  (connect CopyAttribute_112 Out CopyAttribute_16 Source)
  (connect AttributeFilter_20 OutsideFilter CopyAttribute_16 Target)
  (connect CopyAttribute_112 Out CopyAttribute_16 Target)
  (connect AttributeFilter_14 InsideFilter CopyAttribute_30 Source)
  (connect AttributeFilter_14 InsideFilter CopyAttribute_30 Target)
  (connect CreateAttribute_79 Out CopyAttribute_78 Source)
  (connect AttributeFilter_22 OutsideFilter CopyAttribute_78 Target)
  (connect TransformPoints_19 Out CopyAttribute_78 Target)
  (connect CreateAttribute_89 Out CopyAttribute_91 Source)
  (connect ApplyHierarchy_0 Out CopyAttribute_91 Target)
  (connect CopyAttribute_91 Out CopyAttribute_93 Source)
  (connect CopyAttribute_91 Out CopyAttribute_93 Target)
  (connect AttributeFilter_110 OutsideFilter CopyAttribute_112 Source)
  (connect AttributeFilter_110 OutsideFilter CopyAttribute_112 Target)
  (connect NormalToDensity_4 Out DensityFilter_2 In)
  (connect NormalToDensity_31 Out DensityFilter_29 In)
  (connect ApplyHierarchy_1 Out Gather_0 In)
  (connect ApplyHierarchy_2 Out Gather_0 In)
  (connect AttributeFilter_11 OutsideFilter Gather_1 In)
  (connect AttributeFilter_20 InsideFilter Gather_1 In)
  (connect CopyAttribute_16 Out Gather_1 In)
  (connect AttributeFilter_6 InsideFilter MakeRotatorAttribute_7 Y)
  (connect AttributeFilter_9 InsideFilter MakeRotatorAttribute_13 X)
  (connect AttributeMathsOp_26 Out Merge_2 In)
  (connect PointFilter_6 InsideFilter Merge_2 In)
  (connect PointFilter_16 OutsideFilter Merge_2 In)
  (connect PointFilter_59 InsideFilter Merge_2 In)
  (connect Branch_60 Output%20B Merge_3 In)
  (connect CopyAttribute_10 Out Merge_3 In)
  (connect CopyAttribute_78 Out Merge_3 In)
  (connect PointFilter_49 OutsideFilter Merge_3 In)
  (connect AttributeFilter_1 OutsideFilter Merge_25 In)
  (connect AttributeMathsOp_7 Out Merge_25 In)
  (connect AttributeMathsOp_12 Out Merge_25 In)
  (connect AttributeMathsOp_18 Out Merge_25 In)
  (connect AttributeMathsOp_21 Out Merge_25 In)
  (connect AttributeMathsOp_29 Out Merge_25 In)
  (connect AttributeMathsOp_31 Out Merge_25 In)
  (connect DensityFilter_29 Out Merge_37 In)
  (connect NormalToDensity_29 Out Merge_37 In)
  (connect AttributeNoise_6 Out Merge_57 In)
  (connect AttributeNoise_12 Out Merge_57 In)
  (connect AttributeNoise_18 Out Merge_57 In)
  (connect AttributeNoise_24 Out Merge_57 In)
  (connect PointFilter_47 InsideFilter MergePoints_3 In)
  (connect PointFilter_49 InsideFilter MergePoints_3 In)
  (connect AttributeCompareOp_22 Out MergePoints_44 In)
  (connect AttributeFilter_14 OutsideFilter MergePoints_44 In)
  (connect AttributeFilter_0 InsideFilter NormalToDensity_4 In)
  (connect PointFilter_36 OutsideFilter NormalToDensity_29 In)
  (connect PointFilter_36 InsideFilter NormalToDensity_31 In)
  (connect PointFilter_24 InsideFilter PointFilter_1 In)
  (connect PointFilter_22 OutsideFilter PointFilter_4 In)
  (connect AttributeFilter_5 OutsideFilter PointFilter_6 In)
  (connect TransformPoints_3 Out PointFilter_6 In)
  (connect PointFilter_4 OutsideFilter PointFilter_10 In)
  (connect Subgraph_80 Points%20Out PointFilter_11 In)
  (connect PointFilter_10 OutsideFilter PointFilter_16 In)
  (connect PointFilter_59 OutsideFilter PointFilter_22 In)
  (connect Subgraph_0 Points%20Out PointFilter_24 In)
  (connect Projection_41 Out PointFilter_36 In)
  (connect PointFilter_11 OutsideFilter PointFilter_45 In)
  (connect TransformPoints_13 Out PointFilter_45 In)
  (connect CopyAttribute_93 Out PointFilter_47 In)
  (connect PointFilter_47 OutsideFilter PointFilter_49 In)
  (connect PointFilter_6 OutsideFilter PointFilter_59 In)
  (connect CopyAttribute_14 Out Projection_13 In)
  (connect Landscape%20Input Out Projection_13 Projection%20Target)
  (connect AttributeFilter_14 InsideFilter Projection_18 In)
  (connect Landscape%20Input Out Projection_18 Projection%20Target)
  (connect CopyAttribute_7 Out Projection_41 In)
  (connect Landscape%20Input Out Projection_41 Projection%20Target)
  (connect Merge_3 Out Subgraph_0 Points%20In)
  (connect CopyAttribute_0 Out Subgraph_80 Points%20In)
  (connect MakeRotatorAttribute_7 Out TransformPoints_1 In)
  (connect AttributeFilter_17 InsideFilter TransformPoints_2 In)
  (connect AttributeFilter_5 InsideFilter TransformPoints_3 In)
  (connect MakeRotatorAttribute_13 Out TransformPoints_5 In)
  (connect PointFilter_11 InsideFilter TransformPoints_13 In)
  (connect PointFilter_45 InsideFilter TransformPoints_16 In)
  (connect AttributeFilter_22 InsideFilter TransformPoints_19 In)
  (connect AttributeFilter_16 InsideFilter TransformPoints_23 In)
  (connect AttributeFilter_30 InsideFilter TransformPoints_28 In)
  (output Points%20Out MergePoints_44))
