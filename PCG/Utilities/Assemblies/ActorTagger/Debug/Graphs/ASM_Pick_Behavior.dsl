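; PCG utility subgraph that randomly thins assembly points tagged Pick1 to Pick4, keeping one to four respectively and passing untagged points through.
; Each branch picks per ASM_ID partition, subdividing PickLocal points by ParentIndex.
; Density noise and sorting shuffle each set before Random Choice, seeded from a point of the branch's merged set.
; The Pick2 PickLocal filter, unlike the others, leaves its operator at default instead of Equal.
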(graph ASM_Pick_Behavior
  (inputs (Points%20In :type Point))
  (outputs (Points%20Out :type Point))
  ;[0 0 0] [TAG] Pick1 Pick2 Pick3 Pick4
  (bind AttributeFilter_0 (Filter%20Attribute%20Elements :AttributeTypes (struct :Type Boolean :BoolValue true) :bUseConstantThreshold true :bUseSpatialQuery true :bWarnOnDataMissingAttribute false :Operator Equal :TargetAttribute Pick1 :ThresholdAttribute @Last))
  (bind AttributeFilter_3 (Filter%20Attribute%20Elements :AttributeTypes (struct :Type Boolean :BoolValue true) :bUseConstantThreshold true :bUseSpatialQuery true :bWarnOnDataMissingAttribute false :Operator Equal :TargetAttribute Pick2 :ThresholdAttribute @Last))
  (bind AttributeFilter_5 (Filter%20Attribute%20Elements :AttributeTypes (struct :Type Boolean :BoolValue true) :bUseConstantThreshold true :bUseSpatialQuery true :bWarnOnDataMissingAttribute false :Operator Equal :TargetAttribute Pick3 :ThresholdAttribute @Last))
  (bind AttributeFilter_7 (Filter%20Attribute%20Elements :AttributeTypes (struct :Type Boolean :BoolValue true) :bUseConstantThreshold true :bUseSpatialQuery true :bWarnOnDataMissingAttribute false :Operator Equal :TargetAttribute Pick4 :ThresholdAttribute @Last))
  (bind AttributePartition_11 (Attribute%20Partition :PartitionAttributeSelectors [ASM_ID]))
  (bind AttributeFilter_1 (Filter%20Attribute%20Elements :AttributeTypes (struct :Type Boolean :BoolValue true) :bUseConstantThreshold true :bUseSpatialQuery true :bWarnOnDataMissingAttribute false :Operator Equal :TargetAttribute PickLocal :ThresholdAttribute @Last))
  (bind AttributePartition_3 (Attribute%20Partition :PartitionAttributeSelectors [ParentIndex]))
  (bind AttributeNoise_15 (Attribute%20Noise))
  (bind SortAttributes_16 (Sort%20Attributes :bUseStableSort false :SortAttributes [(struct :InputSource $Density)]))
  (bind MergePoints_27 (Merge%20Points))
  (bind GetAttributeFromPointIndex_17 (Get%20Attribute%20From%20Point%20Index :bForceOutputAttributeToBeInElementsDomain true :InputSource $Seed :OutputAttributeName Seed))
  (bind RandomChoice_2 (Random%20Choice :bOutputDiscardedEntries false))
  (bind AttributePartition_12 (Attribute%20Partition :PartitionAttributeSelectors [ASM_ID]))
  (bind AttributeFilter_8 (Filter%20Attribute%20Elements :AttributeTypes (struct :Type Boolean :BoolValue true) :bUseConstantThreshold true :bUseSpatialQuery true :bWarnOnDataMissingAttribute false :TargetAttribute PickLocal :ThresholdAttribute @Last))
  (bind AttributePartition_5 (Attribute%20Partition :PartitionAttributeSelectors [ParentIndex]))
  (bind AttributeNoise_18 (Attribute%20Noise))
  (bind SortAttributes_19 (Sort%20Attributes :bUseStableSort false :SortAttributes [(struct :InputSource $Density)]))
  (bind MergePoints_28 (Merge%20Points))
  (bind GetAttributeFromPointIndex_20 (Get%20Attribute%20From%20Point%20Index :bForceOutputAttributeToBeInElementsDomain true :InputSource $Seed :OutputAttributeName Seed))
  (bind RandomChoice_4 (Random%20Choice :bOutputDiscardedEntries false :FixedNumber 2))
  (bind AttributePartition_13 (Attribute%20Partition :PartitionAttributeSelectors [ASM_ID]))
  (bind AttributeFilter_10 (Filter%20Attribute%20Elements :AttributeTypes (struct :Type Boolean :BoolValue true) :bUseConstantThreshold true :bUseSpatialQuery true :bWarnOnDataMissingAttribute false :Operator Equal :TargetAttribute PickLocal :ThresholdAttribute @Last))
  (bind AttributePartition_15 (Attribute%20Partition :PartitionAttributeSelectors [ParentIndex]))
  (bind AttributeNoise_21 (Attribute%20Noise))
  (bind SortAttributes_22 (Sort%20Attributes :bUseStableSort false :SortAttributes [(struct :InputSource $Density)]))
  (bind MergePoints_29 (Merge%20Points))
  (bind GetAttributeFromPointIndex_23 (Get%20Attribute%20From%20Point%20Index :bForceOutputAttributeToBeInElementsDomain true :InputSource $Seed :OutputAttributeName Seed))
  (bind RandomChoice_6 (Random%20Choice :bOutputDiscardedEntries false :FixedNumber 3))
  (bind AttributePartition_14 (Attribute%20Partition :PartitionAttributeSelectors [ASM_ID]))
  (bind AttributeFilter_17 (Filter%20Attribute%20Elements :AttributeTypes (struct :Type Boolean :BoolValue true) :bUseConstantThreshold true :bUseSpatialQuery true :bWarnOnDataMissingAttribute false :Operator Equal :TargetAttribute PickLocal :ThresholdAttribute @Last))
  (bind AttributePartition_19 (Attribute%20Partition :PartitionAttributeSelectors [ParentIndex]))
  (bind AttributeNoise_24 (Attribute%20Noise))
  (bind SortAttributes_25 (Sort%20Attributes :bUseStableSort false :SortAttributes [(struct :InputSource $Density)]))
  (bind MergePoints_30 (Merge%20Points))
  (bind GetAttributeFromPointIndex_26 (Get%20Attribute%20From%20Point%20Index :bForceOutputAttributeToBeInElementsDomain true :InputSource $Seed :OutputAttributeName Seed))
  (bind RandomChoice_8 (Random%20Choice :bOutputDiscardedEntries false :FixedNumber 4))
  (bind MergePoints_1 (Merge%20Points))
  (connect DefaultInputNode Points%20In AttributeFilter_0 In)
  (connect AttributePartition_11 Out AttributeFilter_1 In)
  (connect AttributeFilter_0 OutsideFilter AttributeFilter_3 In)
  (connect AttributeFilter_3 OutsideFilter AttributeFilter_5 In)
  (connect AttributeFilter_5 OutsideFilter AttributeFilter_7 In)
  (connect AttributePartition_12 Out AttributeFilter_8 In)
  (connect AttributePartition_13 Out AttributeFilter_10 In)
  (connect AttributePartition_14 Out AttributeFilter_17 In)
  (connect AttributeFilter_1 OutsideFilter AttributeNoise_15 In)
  (connect AttributePartition_3 Out AttributeNoise_15 In)
  (connect AttributeFilter_8 OutsideFilter AttributeNoise_18 In)
  (connect AttributePartition_5 Out AttributeNoise_18 In)
  (connect AttributeFilter_10 OutsideFilter AttributeNoise_21 In)
  (connect AttributePartition_15 Out AttributeNoise_21 In)
  (connect AttributeFilter_17 OutsideFilter AttributeNoise_24 In)
  (connect AttributePartition_19 Out AttributeNoise_24 In)
  (connect AttributeFilter_1 InsideFilter AttributePartition_3 In)
  (connect AttributeFilter_8 InsideFilter AttributePartition_5 In)
  (connect AttributeFilter_0 InsideFilter AttributePartition_11 In)
  (connect AttributeFilter_3 InsideFilter AttributePartition_12 In)
  (connect AttributeFilter_5 InsideFilter AttributePartition_13 In)
  (connect AttributeFilter_7 InsideFilter AttributePartition_14 In)
  (connect AttributeFilter_10 InsideFilter AttributePartition_15 In)
  (connect AttributeFilter_17 InsideFilter AttributePartition_19 In)
  (connect MergePoints_27 Out GetAttributeFromPointIndex_17 In)
  (connect MergePoints_28 Out GetAttributeFromPointIndex_20 In)
  (connect MergePoints_29 Out GetAttributeFromPointIndex_23 In)
  (connect MergePoints_30 Out GetAttributeFromPointIndex_26 In)
  (connect AttributeFilter_7 OutsideFilter MergePoints_1 In)
  (connect RandomChoice_2 Chosen MergePoints_1 In)
  (connect RandomChoice_4 Chosen MergePoints_1 In)
  (connect RandomChoice_6 Chosen MergePoints_1 In)
  (connect RandomChoice_8 Chosen MergePoints_1 In)
  (connect SortAttributes_16 Out MergePoints_27 In)
  (connect SortAttributes_19 Out MergePoints_28 In)
  (connect SortAttributes_22 Out MergePoints_29 In)
  (connect SortAttributes_25 Out MergePoints_30 In)
  (connect SortAttributes_16 Out RandomChoice_2 In)
  (connect GetAttributeFromPointIndex_17 Attribute RandomChoice_2 Seed)
  (connect SortAttributes_19 Out RandomChoice_4 In)
  (connect GetAttributeFromPointIndex_20 Attribute RandomChoice_4 Seed)
  (connect SortAttributes_22 Out RandomChoice_6 In)
  (connect GetAttributeFromPointIndex_23 Attribute RandomChoice_6 Seed)
  (connect SortAttributes_25 Out RandomChoice_8 In)
  (connect GetAttributeFromPointIndex_26 Attribute RandomChoice_8 Seed)
  (connect AttributeNoise_15 Out SortAttributes_16 In)
  (connect AttributeNoise_18 Out SortAttributes_19 In)
  (connect AttributeNoise_21 Out SortAttributes_22 In)
  (connect AttributeNoise_24 Out SortAttributes_25 In)
  (output Points%20Out MergePoints_1))
