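; PCG utility graph for the assembly actor tagger that builds the template of predefined tag attributes with their default values.
; A single created point passes through chained Add Attribute stages adding placement, rotation, scale, clutter, pick, axis and ASM_ID tags, then rendering, culling, collision and HLOD properties.
; It outputs the tagged point and an attribute-set copy as RequiredAttributes.
; ASM_ID defaults to 0 because -1 breaks the post-offset copy.
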(graph ASM_Populate_PredefinedTags
  (outputs (Point%20Tags%20Out :type Point) (RequiredAttributes :type Param))
  (bind CreateAttribute_1 (Create%20Constant :AttributeTypes (struct :Type Boolean) :OutputTarget Cliff))
  (bind CreateAttribute_2 (Create%20Constant :AttributeTypes (struct :Type Float) :OutputTarget FloatZ))
  (bind CreateAttribute_4 (Create%20Constant :AttributeTypes (struct :Type Float) :OutputTarget OffsetZ))
  (bind CreateAttribute_8 (Create%20Constant :AttributeTypes (struct :Type Boolean) :OutputTarget RotZ))
  (bind CreateAttribute_9 (Create%20Constant :AttributeTypes (struct :Type Boolean) :OutputTarget RotX))
  (bind CreateAttribute_11 (Create%20Constant :AttributeTypes (struct :Type Boolean) :OutputTarget RandScale10))
  (bind CreateAttribute_13 (Create%20Constant :AttributeTypes (struct :Type Boolean) :OutputTarget RandScale5))
  (bind CreateAttribute_14 (Create%20Constant :AttributeTypes (struct :Type Boolean) :OutputTarget RandScale15))
  (bind CreateAttribute_15 (Create%20Constant :AttributeTypes (struct :Type Boolean) :OutputTarget RandScale20))
  (bind CreateAttribute_18 (Create%20Constant :AttributeTypes (struct :Type Boolean) :OutputTarget IPR))
  (bind CreateAttribute_19 (Create%20Constant :AttributeTypes (struct :Type Boolean) :OutputTarget STL))
  (bind CreateAttribute_20 (Create%20Constant :AttributeTypes (struct :Type Boolean) :OutputTarget RTL))
  (bind CreateAttribute_21 (Create%20Constant :AttributeTypes (struct :Type Boolean) :OutputTarget IPS))
  (bind CreateAttribute_24 (Create%20Constant :AttributeTypes (struct :Type Boolean) :OutputTarget ClutterMed))
  (bind CreateAttribute_25 (Create%20Constant :AttributeTypes (struct :Type Boolean) :OutputTarget Clutter))
  (bind CreateAttribute_26 (Create%20Constant :AttributeTypes (struct :Type Boolean) :OutputTarget ClutterLow))
  (bind CreateAttribute_27 (Create%20Constant :AttributeTypes (struct :Type Boolean) :OutputTarget ClutterHigh))
  (bind CreateAttribute_51 (Create%20Constant :AttributeTypes (struct :Type Boolean) :OutputTarget CULL))
  (bind CreateAttribute_35 (Create%20Constant :AttributeTypes (struct :Type Boolean :Int32Value -1 :StringValue "None") :OutputTarget Pick1))
  (bind CreateAttribute_37 (Create%20Constant :AttributeTypes (struct :Type Boolean :Int32Value -1 :StringValue "None") :OutputTarget Pick2))
  (bind CreateAttribute_39 (Create%20Constant :AttributeTypes (struct :Type Boolean :Int32Value -1 :StringValue "None") :OutputTarget Pick3))
  (bind CreateAttribute_41 (Create%20Constant :AttributeTypes (struct :Type Boolean :Int32Value -1 :StringValue "None") :OutputTarget Pick4))
  (bind CreateAttribute_68 (Create%20Constant :AttributeTypes (struct :Type Boolean) :OutputTarget RotY))
  (bind AddAttribute_64 (Add%20Attribute :bCopyAllAttributes true))
  (bind AddAttribute_67 (Add%20Attribute :bCopyAllAttributes true))
  (bind AddAttribute_69 (Add%20Attribute :bCopyAllAttributes true))
  (bind CreateConstant_0 (Create%20Constant :AttributeTypes (struct :Type Boolean) :OutputTarget FixedScale))
  (bind AddAttribute_71 (Add%20Attribute :bCopyAllAttributes true))
  (bind AddAttribute_73 (Add%20Attribute :bCopyAllAttributes true))
  (bind AddAttribute_77 (Add%20Attribute :bCopyAllAttributes true))
  (bind CreateAttribute_45 (Create%20Constant :AttributeTypes (struct :Type Boolean :StringValue "None") :OutputTarget WZAbove0))
  (bind CreateAttribute_53 (Create%20Constant :AttributeTypes (struct :Type Boolean :StringValue "None") :OutputTarget XUp))
  (bind CreateAttribute_57 (Create%20Constant :AttributeTypes (struct :Type Boolean :StringValue "None") :OutputTarget YUp))
  (bind CreateAttribute_58 (Create%20Constant :AttributeTypes (struct :Type Boolean :StringValue "None") :OutputTarget PickLocal))
  (bind CreateAttribute_60 (Create%20Constant :AttributeTypes (struct :Type Boolean :StringValue "None") :OutputTarget ZAboveLand))
  (bind CreateAttribute_65 (Create%20Constant :AttributeTypes (struct :Type Boolean :StringValue "None") :OutputTarget ZDown))
  (bind AddAttribute_2 (Add%20Attribute :bCopyAllAttributes true))
  ;[0 0 0 0.5] Default Changed from -1 to 0 or post offset copy fails
  (bind CreateAttribute_32 (Create%20Constant :AttributeTypes (struct :Type Integer64 :StringValue "None") :OutputTarget ASM_ID))
  ;[0 0 0] Initial point
  (bind CreatePoints_65 (Create%20Points :PointsToCreate [(struct)]))
  ;[0 0 0 1 25] OUTPUT
  (bind AddAttribute_1 (Add%20Attribute :bCopyAllAttributes true))
  (bind PointToAttributeSet_0 (Point%20To%20Attribute%20Set))
  ;[0 0 0 0.5] Rendering Tags
  (bind CreateConstant_1 (Create%20Constant :AttributeTypes (struct :Type Boolean :StringValue "None" :BoolValue true) :OutputTarget bCastShadow))
  (bind CreateConstant_2 (Create%20Constant :AttributeTypes (struct :Type Float :StringValue "None") :OutputTarget WorldPositionOffsetDisableDistance))
  (bind CreateConstant_3 (Create%20Constant :AttributeTypes (struct :StringValue "None") :OutputTarget DetailMode))
  (bind CreateConstant_4 (Create%20Constant :AttributeTypes (struct :Type Boolean :StringValue "None" :BoolValue true) :OutputTarget bRenderInMainPass))
  (bind CreateConstant_5 (Create%20Constant :AttributeTypes (struct :Type Boolean) :OutputTarget bCastHiddenShadow))
  (bind CreateConstant_6 (Create%20Constant :OutputTarget InstanceStartCullDistance))
  (bind CreateConstant_7 (Create%20Constant :OutputTarget InstanceEndCullDistance))
  (bind CreateConstant_8 (Create%20Constant :AttributeTypes (struct :Type Boolean :BoolValue true) :OutputTarget bAffectDistanceFieldLighting))
  (bind CreateConstant_9 (Create%20Constant :AttributeTypes (struct :Type Boolean :BoolValue true) :OutputTarget bAffectDynamicIndirectLighting))
  (bind CreateConstant_10 (Create%20Constant :AttributeTypes (struct :Type Boolean :BoolValue true) :OutputTarget bAffectDynamicIndirectLighting))
  (bind CreateConstant_11 (Create%20Constant :AttributeTypes (struct :Type Boolean :BoolValue true) :OutputTarget bVisibleInRayTracing))
  (bind CreateConstant_12 (Create%20Constant :AttributeTypes (struct :Type Boolean :BoolValue true) :OutputTarget bUseDefaultCollision))
  (bind CreateConstant_13 (Create%20Constant :AttributeTypes (struct :Type Boolean) :OutputTarget bIncludeInHLOD))
  (bind AddAttribute_0 (Add%20Attribute :bCopyAllAttributes true))
  (reroute Final%20Tags AddAttribute_0)
  (reroute Point CreatePoints_65)
  (connect CreateConstant_1 Out AddAttribute_0 Attributes)
  (connect CreateConstant_2 Out AddAttribute_0 Attributes)
  (connect CreateConstant_3 Out AddAttribute_0 Attributes)
  (connect CreateConstant_4 Out AddAttribute_0 Attributes)
  (connect CreateConstant_5 Out AddAttribute_0 Attributes)
  (connect CreateConstant_6 Out AddAttribute_0 Attributes)
  (connect CreateConstant_7 Out AddAttribute_0 Attributes)
  (connect CreateConstant_8 Out AddAttribute_0 Attributes)
  (connect CreateConstant_9 Out AddAttribute_0 Attributes)
  (connect CreateConstant_10 Out AddAttribute_0 Attributes)
  (connect CreateConstant_11 Out AddAttribute_0 Attributes)
  (connect CreateConstant_12 Out AddAttribute_0 Attributes)
  (connect CreateConstant_13 Out AddAttribute_0 Attributes)
  (connect AddAttribute_2 Out AddAttribute_0 In)
  (connect CreateAttribute_45 Out AddAttribute_2 Attributes)
  (connect CreateAttribute_53 Out AddAttribute_2 Attributes)
  (connect CreateAttribute_57 Out AddAttribute_2 Attributes)
  (connect CreateAttribute_58 Out AddAttribute_2 Attributes)
  (connect CreateAttribute_60 Out AddAttribute_2 Attributes)
  (connect CreateAttribute_65 Out AddAttribute_2 Attributes)
  (connect AddAttribute_77 Out AddAttribute_2 In)
  (connect CreateAttribute_1 Out AddAttribute_64 Attributes)
  (connect CreateAttribute_2 Out AddAttribute_64 Attributes)
  (connect CreateAttribute_4 Out AddAttribute_64 Attributes)
  (connect CreateAttribute_51 Out AddAttribute_64 Attributes)
  (connect CreatePoints_65 Out AddAttribute_64 In)
  (connect CreateAttribute_8 Out AddAttribute_67 Attributes)
  (connect CreateAttribute_9 Out AddAttribute_67 Attributes)
  (connect CreateAttribute_68 Out AddAttribute_67 Attributes)
  (connect AddAttribute_64 Out AddAttribute_67 In)
  (connect CreateAttribute_18 Out AddAttribute_69 Attributes)
  (connect CreateAttribute_19 Out AddAttribute_69 Attributes)
  (connect CreateAttribute_20 Out AddAttribute_69 Attributes)
  (connect CreateAttribute_21 Out AddAttribute_69 Attributes)
  (connect AddAttribute_67 Out AddAttribute_69 In)
  (connect CreateAttribute_11 Out AddAttribute_71 Attributes)
  (connect CreateAttribute_13 Out AddAttribute_71 Attributes)
  (connect CreateAttribute_14 Out AddAttribute_71 Attributes)
  (connect CreateAttribute_15 Out AddAttribute_71 Attributes)
  (connect CreateConstant_0 Out AddAttribute_71 Attributes)
  (connect AddAttribute_69 Out AddAttribute_71 In)
  (connect CreateAttribute_24 Out AddAttribute_73 Attributes)
  (connect CreateAttribute_25 Out AddAttribute_73 Attributes)
  (connect CreateAttribute_26 Out AddAttribute_73 Attributes)
  (connect CreateAttribute_27 Out AddAttribute_73 Attributes)
  (connect AddAttribute_71 Out AddAttribute_73 In)
  (connect CreateAttribute_32 Out AddAttribute_77 Attributes)
  (connect CreateAttribute_35 Out AddAttribute_77 Attributes)
  (connect CreateAttribute_37 Out AddAttribute_77 Attributes)
  (connect CreateAttribute_39 Out AddAttribute_77 Attributes)
  (connect CreateAttribute_41 Out AddAttribute_77 Attributes)
  (connect AddAttribute_73 Out AddAttribute_77 In)
  (connect Final%20Tags Out PointToAttributeSet_0 In)
  (output Point%20Tags%20Out Final%20Tags)
  (output RequiredAttributes PointToAttributeSet_0))
