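; PCG debug subgraph for ActorTagger assemblies that copies source points onto target points while handling hierarchy data.
; Sources get a constant ASM_ID, and targets are prepopulated with required attributes.
; By default, IgnoreTargetPointXYRotations rebuilds target rotations from their local up axis and randomizes yaw before copying.
; Copy Points skips the every-source-on-every-target mode, then SG_PostCopyPoints_OffsetIndices post-processes the result using the targets, for point output.
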(graph SG_CopyPointsWithHierarchy
  (params (IgnoreTargetPointXYRotations :type Bool :default true))
  (inputs (CopyPoints%20Source :type Point) (CopyPoints%20Target :type Point) (RequiredAttributes :type Param))
  (outputs (Points :type Point))
  (bind CreateAttribute_0 (Create%20Constant :AttributeTypes (struct :Type Integer64) :OutputTarget ASM_ID))
  (bind AddAttribute_0 (Add%20Attribute :OutputTarget @Source))
  (bind AddAttribute_2 (Add%20Attribute :bCopyAllAttributes true)) ; Prepopulate Required Assembly Attributes
  (bind CopyPoints_0 (Copy%20Points :bCopyEachSourceOnEveryTarget false))
  (bind Subgraph_1 (Subgraph :Graph /PCG/Utilities/Assemblies/ActorTagger/Debug/Graphs/SG_PostCopyPoints_OffsetIndices.SG_PostCopyPoints_OffsetIndices))
  (bind ToPoint_3 (To%20Point))
  ;[0 0 0] Ignore Target Point XY Rotations
  (bind Branch_0 (Branch :bOutputToB-param IgnoreTargetPointXYRotations))
  (bind CopyAttribute_19 (Copy%20Attributes :InputSource $Rotation.Up :OutputTarget LocalUpAxis))
  (bind MakeRotatorAttribute_3 (Make%20Rotator%20Attribute :InputSource1 LocalUpAxis :Operation MakeRotFromZX :OutputTarget $Rotation))
  (bind TransformPoints_0 (Transform%20Points :bRecomputeSeed true :RotationMax (struct :Yaw 3.6e+02) :RotationMin (struct :Yaw -3.6e+02)))
  (bind Gather_0 (Gather))
  (connect CreateAttribute_0 Out AddAttribute_0 Attributes)
  (connect DefaultInputNode CopyPoints%20Source AddAttribute_0 In)
  (connect DefaultInputNode RequiredAttributes AddAttribute_2 Attributes)
  (connect DefaultInputNode CopyPoints%20Target AddAttribute_2 In)
  (connect AddAttribute_2 Out Branch_0 In)
  (connect Branch_0 Output%20B CopyAttribute_19 Source)
  (connect Branch_0 Output%20B CopyAttribute_19 Target)
  (connect AddAttribute_0 Out CopyPoints_0 Source)
  (connect Gather_0 Out CopyPoints_0 Target)
  (connect Branch_0 Output%20A Gather_0 In)
  (connect TransformPoints_0 Out Gather_0 In)
  (connect CopyAttribute_19 Out MakeRotatorAttribute_3 X)
  (connect CopyAttribute_19 Out MakeRotatorAttribute_3 Z)
  (connect CopyPoints_0 Out Subgraph_1 CopyPointsOut)
  (connect Gather_0 Out Subgraph_1 CopyPointsTargetIn)
  (connect Subgraph_1 Lerp%20Out ToPoint_3 In)
  (connect MakeRotatorAttribute_3 Out TransformPoints_0 In)
  (output Points ToPoint_3))
