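; PCG debug subgraph for the ActorTagger assembly utilities that offsets ActorIndex and ParentIndex on points after a Copy Points step.
; It derives a per-point Offset from point index modulo target count, multiplied by the copied-point count, and adds it to both indices.
; The sign of the original ParentIndex plus one drives a lerp that keeps -1 for parentless points, avoiding a branch, before Offset is deleted.
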(graph SG_PostCopyPoints_OffsetIndices
  (inputs (CopyPointsOut :type Point) (CopyPointsTargetIn :type Point))
  (outputs (Lerp%20Out :type Point))
  (bind GetPointsCount_1 (Get%20Element%20Count))
  (bind AttributeMathsOp_6 (Attribute%20Maths%20Op :InputSource1 $Index :Operation Modulo :OutputTarget Offset))
  (bind GetPointsCount_0 (Get%20Element%20Count))
  (bind AttributeMathsOp_0 (Attribute%20Maths%20Op :InputSource1 Offset :Operation Multiply :OutputTarget Offset))
  (bind AttributeMathsOp_1 (Attribute%20Maths%20Op :InputSource1 Offset :InputSource2 ActorIndex :OutputTarget ActorIndex))
  (bind AttributeMathsOp_2 (Attribute%20Maths%20Op :InputSource1 Offset :InputSource2 ParentIndex :OutputTarget ParentIndex))
  (bind CreateConstant_0 (Create%20Constant :AttributeTypes (struct :Type Integer32 :Int32Value 1)))
  (bind AttributeMathsOp_4 (Attribute%20Maths%20Op :InputSource1 ParentIndex))
  (bind AttributeMathsOp_3 (Attribute%20Maths%20Op :InputSource1 ParentIndex :Operation Sign :OutputTarget HasParent))
  (bind CreateConstant_1 (Create%20Constant :AttributeTypes (struct :DoubleValue -1.0)))
  (bind AttributeMathsOp_5 (Attribute%20Maths%20Op :InputSource3 HasParent :Operation Lerp))
  (bind DeleteAttributes_0 (Delete%20Attributes :SelectedAttributes "Offset"))
  (connect AttributeMathsOp_6 Out AttributeMathsOp_0 InA)
  (connect GetPointsCount_0 Out AttributeMathsOp_0 InB)
  (connect AttributeMathsOp_0 Out AttributeMathsOp_1 InA)
  (connect AttributeMathsOp_0 Out AttributeMathsOp_1 InB)
  (connect AttributeMathsOp_1 Out AttributeMathsOp_2 InA)
  (connect AttributeMathsOp_1 Out AttributeMathsOp_2 InB)
  (connect AttributeMathsOp_4 Out AttributeMathsOp_3 In)
  (connect AttributeMathsOp_1 Out AttributeMathsOp_4 InA)
  (connect CreateConstant_0 Out AttributeMathsOp_4 InB)
  (connect CreateConstant_1 Out AttributeMathsOp_5 InA)
  (connect AttributeMathsOp_2 Out AttributeMathsOp_5 InB)
  (connect AttributeMathsOp_3 Out AttributeMathsOp_5 Ratio)
  (connect DefaultInputNode CopyPointsOut AttributeMathsOp_6 InA)
  (connect GetPointsCount_1 Out AttributeMathsOp_6 InB)
  (connect AttributeMathsOp_5 Out DeleteAttributes_0 In)
  (connect DefaultInputNode CopyPointsOut GetPointsCount_0 In)
  (connect DefaultInputNode CopyPointsTargetIn GetPointsCount_1 In)
  (output Lerp%20Out DeleteAttributes_0))
